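# One-shot stereo MetaSound parent for fall damage hit notifications, playing a random wave after an optional delay.
# DamageDealt maps to a clamped volume scaled by modulator and gain, while accumulated instigator damage drives pitch through a wavetable curve.
# The Mobile page uses mono playback, skips the wavetable and trims 3 dB.
# Bus sends feed sidechains for other game audio and instigator hitmarkers.
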MetaSoundSource /Game/Athena/Sounds/HitNotifications/Gunplay_Update/MSS_HitNotifies_GunplayUpdate_FallDamage_Parent
inputs [
 HitMarker.DamageDealt {type=Float}
 HitMarker.InstigatorAccumulatedDamage {type=Float}
 LayerA.Pitch.Max {type=Float}
 LayerA.Pitch.Min {type=Float}
 LayerA.PitchTable {type=WaveTableBankAsset}
 LayerA.Wave {type=WaveAsset:Array}
 Patch.DamageDealthVolModifier.Max {type=Float}
 Patch.DamageDealthVolModifier.Min {type=Float}
 Patch.DamageDealthVolModifier.ReClamp.Max {type=Float}
 Patch.DamageDealthVolModifier.ReClamp.Min {type=Float}
 Patch.Gain {type=Float}
 Patch.Modulator.Volume {type=Modulator}
 Patch.OptionalDelay {type=Time}
 UE.Source.OnPlay {type=Trigger}
]
interfaces [UE.OutputFormat.Stereo:1.0, UE.Source.OneShot:1.0, UE.Source:1.1]
outputs [
 "UE.OutputFormat.Stereo.Audio:0" {type=Audio}
 "UE.OutputFormat.Stereo.Audio:1" {type=Audio}
 UE.Source.OneShot.OnFinished {type=Trigger}
]
pages [
 Default {
  is_build_graph=true
  input_defaults {
   HitMarker.DamageDealt=0
   HitMarker.InstigatorAccumulatedDamage=0
   LayerA.Pitch.Max=0
   LayerA.Pitch.Min=0
   LayerA.PitchTable=&/WB_Hitmarker_GunplayUpdate
   Patch.DamageDealthVolModifier.Max=2
   Patch.DamageDealthVolModifier.Min=0.6
   Patch.DamageDealthVolModifier.ReClamp.Max=1.2
   Patch.DamageDealthVolModifier.ReClamp.Min=0
   Patch.Gain=1
   Patch.Modulator.Volume=/Game/Athena/Sounds/HitNotifications/Gunplay_Update/Modulation/Patch_Hitmarkers_Receiver_Standard_Volume
   Patch.OptionalDelay=0
   UE.Source.OnPlay=null
   LayerA.Wave [&/Hitmarker_Instigator_Eliminated_01]
  }
  nodes [
   n1 = MapRange.MapRange:Float(In=^HitMarker.DamageDealt, Out Range B=^Patch.DamageDealthVolModifier.Max, Out Range A=^Patch.DamageDealthVolModifier.Min) {also {defaults {In Range B=100}}}
   n2 = UE.Multiply:Float(
    AdditionalOperands=^Patch.Gain
    PrimaryOperand=UE.Multiply:Float(
     AdditionalOperands=Clamp.Clamp:Float(Max=^Patch.DamageDealthVolModifier.ReClamp.Max, Min=^Patch.DamageDealthVolModifier.ReClamp.Min, In=@n1.Out Value).Value
     PrimaryOperand=Modulation.GetModulatorValue(Modulator=^Patch.Modulator.Volume).Out
    ).Out
   )
   TriggerDelay_0 = UE.Trigger Delay(Delay Time=^Patch.OptionalDelay, In=^UE.Source.OnPlay)
   ArrayRandomGet_0 = Array.Random Get:WaveAsset:Array(In Array=^LayerA.Wave, Next=@TriggerDelay_0.Out) {also {defaults {Weights []}}}
   n3 = MapRange.MapRange:Float(In=^HitMarker.InstigatorAccumulatedDamage) {also {defaults {In Range B=300}}}
   WavePlayer_0 = UE.Wave Player:Stereo(
    Play=@ArrayRandomGet_0.On Next
    Wave Asset=@ArrayRandomGet_0.Value
    Pitch Shift=MapRange.MapRange:Float(
     Out Range B=^LayerA.Pitch.Max
     Out Range A=^LayerA.Pitch.Min
     In=UE.WaveTableBankEvaluate(WaveTableBank=^LayerA.PitchTable, Input=@n3.Out Value) {also {defaults {Index=1}}}.Output
    ).Out Value
   ) {also {version [1,1]}}
   Multiply_0 = UE.Multiply:Audio by Float(AdditionalOperands=@n2.Out, PrimaryOperand=@WavePlayer_0.Out Left)
   Multiply_1 = UE.Multiply:Audio by Float(AdditionalOperands=@n2.Out, PrimaryOperand=@WavePlayer_0.Out Right)
  ]
  links [
   {dst="@UE.OutputFormat.Stereo.Audio:0", src=Multiply_0.Out}
   {dst="@UE.OutputFormat.Stereo.Audio:1", src=Multiply_1.Out}
   {dst="@UE.Source.OneShot.OnFinished", src=WavePlayer_0.On Finished}
  ]
 }
 Mobile {
  nodes [
   DecibelstoLinearGain_0 = UE.Decibels to Linear Gain:Float() {also {defaults {Decibels=-3}}}
   n1 = Clamp.Clamp:Float(
    Max=^Patch.DamageDealthVolModifier.ReClamp.Max
    Min=^Patch.DamageDealthVolModifier.ReClamp.Min
    In=MapRange.MapRange:Float(In=^HitMarker.DamageDealt, Out Range B=^Patch.DamageDealthVolModifier.Max, Out Range A=^Patch.DamageDealthVolModifier.Min) {also {defaults {In Range B=100}}}.Out Value
   )
   n2 = Modulation.GetModulatorValue(Modulator=^Patch.Modulator.Volume)
   Multiply_1 = UE.Multiply:Float(
    AdditionalOperands=@DecibelstoLinearGain_0.Linear Gain
    PrimaryOperand=UE.Multiply:Float(AdditionalOperands=^Patch.Gain, PrimaryOperand=UE.Multiply:Float(AdditionalOperands=@n1.Value, PrimaryOperand=@n2.Out).Out).Out
   )
   TriggerDelay_0 = UE.Trigger Delay(Delay Time=^Patch.OptionalDelay, In=^UE.Source.OnPlay)
   ArrayRandomGet_0 = Array.Random Get:WaveAsset:Array(In Array=^LayerA.Wave, Next=@TriggerDelay_0.Out) {also {defaults {Weights []}}}
   WavePlayer_0 = UE.Wave Player:Mono(
    Play=@ArrayRandomGet_0.On Next
    Wave Asset=@ArrayRandomGet_0.Value
    Pitch Shift=MapRange.MapRange:Float(
     Out Range B=^LayerA.Pitch.Max
     Out Range A=^LayerA.Pitch.Min
     In=MapRange.MapRange:Float(In=^HitMarker.InstigatorAccumulatedDamage) {also {defaults {In Range B=300}}}.Out Value
    ).Out Value
   ) {also {version [1,1]}}
   Multiply_0 = UE.Multiply:Audio by Float(AdditionalOperands=@Multiply_1.Out, PrimaryOperand=@WavePlayer_0.Out Mono)
  ]
  links [
   {dst="@UE.OutputFormat.Stereo.Audio:0", src=Multiply_0.Out}
   {dst="@UE.OutputFormat.Stereo.Audio:1", src=Multiply_0.Out}
   {dst="@UE.Source.OneShot.OnFinished", src=WavePlayer_0.On Finished}
  ]
 }
]
properties {
 assetImportData=&self:AssetImportData
 audioPropertiesSheet=&/APS_Hitmarker_Receiver_Standard
 numChannels=2
 outputFormat=Stereo
 soundClassObject=&Fort_Hitmarkers
 soundSubmixObject=&Fortnite_SFX
 busSends [
  {audioBus=&AB_Sidechain_FortWeapon_Vol, sendLevel=0.3}
  {audioBus=&AB_Sidechain_FortExplosion_Vol, sendLevel=0.4}
  {audioBus=&AB_Sidechain_FortAmbient_Vol, sendLevel=0.3}
  {audioBus=&AB_Sidechain_FortLoot_Vol, sendLevel=0.3}
  {audioBus=&AB_Sidechain_FortVehicles_Vol, sendLevel=0.3}
  {audioBus=&AB_Sidechain_FortCreatures_Vol, sendLevel=0.3}
  {audioBus=&AB_Sidechain_FortEmote_Vol, sendLevel=0.3}
  {audioBus=/Game/Athena/Sounds/HitNotifications/Gunplay_Update/Modulation/AB_Sidechain_Hitmarkers_Instigator_Critical_Bus, sendLevel=0.7}
  {audioBus=/Game/Athena/Sounds/HitNotifications/Gunplay_Update/Modulation/AB_Sidechain_Hitmarkers_Instigator_Standard_Bus, sendLevel=0.7}
 ]
 concurrencySet [&/CON_Hitmarker_GunplayUpdate_Receiver_Standard]
 modulationSettings {
  volumeRouting=Union
  volumeModulationDestination {bEnableModulation=true, modulators [&Patch_FortHitNotify_Vol]}
 }
}
