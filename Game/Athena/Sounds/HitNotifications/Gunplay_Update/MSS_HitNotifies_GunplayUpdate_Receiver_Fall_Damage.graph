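# Stereo one-shot MetaSound preset for the hit notification a player hears when receiving fall damage.
# It inherits its graph from the gunplay-update FallDamage parent, supplying three fall-hit waves, a pitch table, damage-based volume limits and a receiver volume modulator.
# It sends to sidechain buses for weapons, explosions, ambience, other categories and instigator hitmarkers, presumably for ducking.
# Two receiver concurrency sets govern overlapping plays.
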MetaSoundSource /Game/Athena/Sounds/HitNotifications/Gunplay_Update/MSS_HitNotifies_GunplayUpdate_Receiver_Fall_Damage
preset_parent=&/MSS_HitNotifies_GunplayUpdate_FallDamage_Parent
inputs [
 HitMarker.DamageDealt {type=Float}
 HitMarker.InstigatorAccumulatedDamage {type=Float}
 LayerA.Pitch.Max {type=Float}
 LayerA.Pitch.Min {type=Float}
 LayerA.PitchTable {type=WaveTableBankAsset}
 LayerA.Wave {type=WaveAsset:Array}
 Patch.DamageDealthVolModifier.Max {type=Float}
 Patch.DamageDealthVolModifier.Min {type=Float}
 Patch.DamageDealthVolModifier.ReClamp.Max {type=Float}
 Patch.DamageDealthVolModifier.ReClamp.Min {type=Float}
 Patch.Gain {type=Float}
 Patch.Modulator.Volume {type=Modulator}
 Patch.OptionalDelay {type=Time}
 UE.Source.OnPlay {type=Trigger}
]
interfaces [UE.OutputFormat.Stereo:1.0, UE.Source.OneShot:1.0, UE.Source:1.1]
outputs [
 "UE.OutputFormat.Stereo.Audio:0" {type=Audio}
 "UE.OutputFormat.Stereo.Audio:1" {type=Audio}
 UE.Source.OneShot.OnFinished {type=Trigger}
]
pages [Default {input_defaults {
   HitMarker.DamageDealt=0
   HitMarker.InstigatorAccumulatedDamage=0
   LayerA.Pitch.Max=0
   LayerA.Pitch.Min=0
   LayerA.PitchTable=&/WB_Hitmarker_GunplayUpdate
   Patch.DamageDealthVolModifier.Max=2
   Patch.DamageDealthVolModifier.Min=0.6
   Patch.DamageDealthVolModifier.ReClamp.Max=1.2
   Patch.DamageDealthVolModifier.ReClamp.Min=0
   Patch.Gain=1
   Patch.Modulator.Volume=/Game/Athena/Sounds/HitNotifications/Gunplay_Update/Modulation/Patch_Hitmarkers_Receiver_Standard_Volume
   Patch.OptionalDelay=0
   UE.Source.OnPlay=null
   LayerA.Wave [
    &/Hitmarker_Receiver_Fall_01
    &/Hitmarker_Receiver_Fall_02
    &/Hitmarker_Receiver_Fall_03
   ]
  }}]
properties {
 assetImportData=&self:AssetImportData
 audioPropertiesSheet=&/APS_Hitmarker_Receiver_Standard
 numChannels=2
 outputFormat=Stereo
 soundClassObject=&Fort_Hitmarkers
 soundSubmixObject=&Fortnite_SFX
 busSends [
  {audioBus=&AB_Sidechain_FortWeapon_Vol, sendLevel=0.3}
  {audioBus=&AB_Sidechain_FortExplosion_Vol, sendLevel=0.4}
  {audioBus=&AB_Sidechain_FortAmbient_Vol, sendLevel=0.3}
  {audioBus=&AB_Sidechain_FortLoot_Vol, sendLevel=0.3}
  {audioBus=&AB_Sidechain_FortVehicles_Vol, sendLevel=0.3}
  {audioBus=&AB_Sidechain_FortCreatures_Vol, sendLevel=0.3}
  {audioBus=&AB_Sidechain_FortEmote_Vol, sendLevel=0.3}
  {audioBus=/Game/Athena/Sounds/HitNotifications/Gunplay_Update/Modulation/AB_Sidechain_Hitmarkers_Instigator_Critical_Bus, sendLevel=0.7}
  {audioBus=/Game/Athena/Sounds/HitNotifications/Gunplay_Update/Modulation/AB_Sidechain_Hitmarkers_Instigator_Standard_Bus, sendLevel=0.7}
 ]
 concurrencySet [
  &/CON_Hitmarker_GunplayUpdate_Receiver_Standard
  /Game/Athena/Sounds/HitNotifications/Gunplay_Update/Settings/CONC_Hitmarker_GunplayUpdate_Receiver_Health
 ]
 modulationSettings {
  volumeRouting=Union
  volumeModulationDestination {bEnableModulation=true, modulators [&Patch_FortHitNotify_Vol]}
 }
}
