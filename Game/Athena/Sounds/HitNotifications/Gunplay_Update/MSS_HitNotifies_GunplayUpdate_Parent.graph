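# Parent MetaSound for gunplay-update hitmarker sounds, a stereo one-shot that layers two randomly picked waves (A and B) after an optional delay.
# Accumulated instigator damage drives layer pitch, while damage dealt and a modulator scale volume.
# Time since last play selects a fade-in for rapid retriggers, and hit azimuth plus spread sets stereo panning.
# The Mobile page mixes in mono, drops the wavetable pitch curves and trims 3 dB.
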MetaSoundSource /Game/Athena/Sounds/HitNotifications/Gunplay_Update/MSS_HitNotifies_GunplayUpdate_Parent
inputs [
 HitMarker.Azimuth {type=Float}
 HitMarker.DamageDealt {type=Float}
 HitMarker.Distance {type=Float}
 HitMarker.InstigatorAccumulatedDamage {type=Float}
 HitMarker.ReceiverPercentHealth {type=Float}
 HitMarker.ReceiverPercentShield {type=Float}
 LayerA.Gain {type=Float}
 LayerA.Pitch.Max {type=Float}
 LayerA.Pitch.Min {type=Float}
 LayerA.PitchTable {type=WaveTableBankAsset}
 LayerA.Wave {type=WaveAsset:Array}
 LayerB.Gain {type=Float}
 LayerB.Pitch.Max {type=Float}
 LayerB.Pitch.Min {type=Float}
 LayerB.PitchTable {type=WaveTableBankAsset}
 LayerB.Wave {type=WaveAsset:Array}
 Patch.DamageDealthVolModifier.Max {type=Float}
 Patch.DamageDealthVolModifier.Min {type=Float}
 Patch.DamageDealthVolModifier.ReClamp.Max {type=Float}
 Patch.DamageDealthVolModifier.ReClamp.Min {type=Float}
 Patch.Gain {type=Float}
 Patch.Modulator.Volume {type=Modulator}
 Patch.OptionalDelay {type=Time}
 Patch.StereoSpread {type=Float}
 Retrigger.BehaviorThresholdTime {type=Float, access=Constructor}
 Retrigger.FadeInTime {type=Float, access=Constructor}
 Time.ElapsedTime {type=Float}
 Time.GameTimeSeconds {type=Float}
 Time.PlayTime {type=Float}
 Time.TimeSinceLastPlayed {type=Float}
 UE.Source.OnPlay {type=Trigger}
]
interfaces [Fort.HitMarker:1.0, Fort.Time:1.0, UE.OutputFormat.Stereo:1.0, UE.Source.OneShot:1.0, UE.Source:1.1]
outputs [
 "UE.OutputFormat.Stereo.Audio:0" {type=Audio}
 "UE.OutputFormat.Stereo.Audio:1" {type=Audio}
 UE.Source.OneShot.OnFinished {type=Trigger}
]
pages [
 Default {
  is_build_graph=true
  input_defaults {
   HitMarker.Azimuth=0
   HitMarker.DamageDealt=0
   HitMarker.Distance=0
   HitMarker.InstigatorAccumulatedDamage=0
   HitMarker.ReceiverPercentHealth=1
   HitMarker.ReceiverPercentShield=1
   LayerA.Gain=1
   LayerA.Pitch.Max=0
   LayerA.Pitch.Min=0
   LayerA.PitchTable=&/WB_Hitmarker_GunplayUpdate
   LayerB.Gain=1.2
   LayerB.Pitch.Max=2
   LayerB.Pitch.Min=-1
   LayerB.PitchTable=&WB_Hitmarker_GunplayUpdate
   Patch.DamageDealthVolModifier.Max=2
   Patch.DamageDealthVolModifier.Min=0.6
   Patch.DamageDealthVolModifier.ReClamp.Max=1.2
   Patch.DamageDealthVolModifier.ReClamp.Min=0
   Patch.Gain=1
   Patch.Modulator.Volume=/Game/Athena/Sounds/HitNotifications/Gunplay_Update/Modulation/Patch_Hitmarkers_Receiver_Standard_Volume
   Patch.OptionalDelay=0
   Patch.StereoSpread=1
   Retrigger.BehaviorThresholdTime=0.71
   Retrigger.FadeInTime=0.1
   Time.ElapsedTime=0
   Time.GameTimeSeconds=0
   Time.PlayTime=0
   Time.TimeSinceLastPlayed=0
   UE.Source.OnPlay=null
   LayerA.Wave [&/Hitmarker_Instigator_Eliminated_01]
   LayerB.Wave []
  }
  nodes [
   n1 = Clamp.Clamp:Float(In=UE.Divide:Float(PrimaryOperand=^HitMarker.Azimuth) {also {defaults {AdditionalOperands=90}}}.Out) {also {defaults {Max=1, Min=-1}}}
   TriggerDelay_0 = UE.Trigger Delay(Delay Time=^Patch.OptionalDelay, In=^UE.Source.OnPlay)
   ArrayRandomGet_0 = Array.Random Get:WaveAsset:Array(In Array=^LayerB.Wave, Next=@TriggerDelay_0.Out)
   n2 = MapRange.MapRange:Float(In=^HitMarker.InstigatorAccumulatedDamage) {also {defaults {In Range B=300}}}
   WavePlayer_0 = UE.Wave Player:Mono(
    Play=@ArrayRandomGet_0.On Next
    Wave Asset=@ArrayRandomGet_0.Value
    Pitch Shift=MapRange.MapRange:Float(
     Out Range B=^LayerB.Pitch.Max
     Out Range A=^LayerB.Pitch.Min
     In=UE.WaveTableBankEvaluate(WaveTableBank=^LayerB.PitchTable, Input=@n2.Out Value) {also {defaults {Index=1}}}.Output
    ).Out Value
   ) {also {version [1,1]}}
   ArrayRandomGet_1 = Array.Random Get:WaveAsset:Array(In Array=^LayerA.Wave, Next=@TriggerDelay_0.Out)
   WavePlayer_1 = UE.Wave Player:Stereo(
    Play=@ArrayRandomGet_1.On Next
    Wave Asset=@ArrayRandomGet_1.Value
    Pitch Shift=MapRange.MapRange:Float(
     Out Range B=^LayerA.Pitch.Max
     Out Range A=^LayerA.Pitch.Min
     In=UE.WaveTableBankEvaluate(WaveTableBank=^LayerA.PitchTable, Input=@n2.Out Value) {also {defaults {Index=1}}}.Output
    ).Out Value
   ) {also {version [1,1]}}
   AudioMixerAudioMixerStereo2_0 = "AudioMixer.Audio Mixer (Stereo, 2)"(
    Gain 0=^LayerA.Gain
    Gain 1=^LayerB.Gain
    In 1 L=@WavePlayer_0.Out Mono
    In 1 R=@WavePlayer_0.Out Mono
    In 0 L=@WavePlayer_1.Out Left
    In 0 R=@WavePlayer_1.Out Right
   )
   n3 = TriggerCompare.Clamp:Float(A=^Time.TimeSinceLastPlayed, Compare=^UE.Source.OnPlay) {also {defaults {B=0}}}
   n4 = TriggerCompare.Clamp:Float(B=^Retrigger.BehaviorThresholdTime, A=^Time.TimeSinceLastPlayed, Compare=@n3.False) {also {defaults {Type=3}}}
   TriggerAnyTriggerAny2_0 = "TriggerAny.Trigger Any (2)"(In 1=@n4.True, In 0=@n3.True)
   TriggerRouteTriggerRouteFloat2_0 = "TriggerRoute.Trigger Route (Float, 2):Float"(Value 1=^Retrigger.FadeInTime, Set 0=@TriggerAnyTriggerAny2_0.Out, Set 1=@n4.False) {also {defaults {Value 0=0}}}
   n5 = Clamp.Clamp:Float(
    Max=^Patch.DamageDealthVolModifier.ReClamp.Max
    Min=^Patch.DamageDealthVolModifier.ReClamp.Min
    In=MapRange.MapRange:Float(In=^HitMarker.DamageDealt, Out Range B=^Patch.DamageDealthVolModifier.Max, Out Range A=^Patch.DamageDealthVolModifier.Min) {also {defaults {In Range B=100}}}.Out Value
   )
   n6 = Modulation.GetModulatorValue(Modulator=^Patch.Modulator.Volume)
   n7 = UE.Multiply:Float(
    AdditionalOperands=&MSP_Fade(In=@TriggerRouteTriggerRouteFloat2_0.On Set, Fade Time=@TriggerRouteTriggerRouteFloat2_0.Value).Value Out
    PrimaryOperand=UE.Multiply:Float(AdditionalOperands=^Patch.Gain, PrimaryOperand=UE.Multiply:Float(AdditionalOperands=@n5.Value, PrimaryOperand=@n6.Out).Out).Out
   )
   StereoPanner_0 = UE.Stereo Panner(
    Pan Amount=Clamp.Clamp:Float(In=UE.Add:Float(AdditionalOperands=^Patch.StereoSpread, PrimaryOperand=@n1.Value).Out) {also {defaults {Max=1, Min=-1}}}.Value
    In=UE.Reroute(Value=UE.Multiply:Audio by Float(PrimaryOperand=@AudioMixerAudioMixerStereo2_0.Out R, AdditionalOperands=@n7.Out).Out).Value
   )
   StereoPanner_1 = UE.Stereo Panner(
    Pan Amount=Clamp.Clamp:Float(In=UE.Subtract:Float(AdditionalOperands=^Patch.StereoSpread, PrimaryOperand=@n1.Value).Out) {also {defaults {Max=1, Min=-1}}}.Value
    In=UE.Multiply:Audio by Float(PrimaryOperand=@AudioMixerAudioMixerStereo2_0.Out L, AdditionalOperands=@n7.Out).Out
   )
   Add_0 = UE.Add:Audio(AdditionalOperands=@StereoPanner_0.Out Left, PrimaryOperand=@StereoPanner_1.Out Left)
   Add_1 = UE.Add:Audio(AdditionalOperands=@StereoPanner_0.Out Right, PrimaryOperand=@StereoPanner_1.Out Right)
   TriggerAccumulatorTriggerAccumulate2_0 = "TriggerAccumulator.Trigger Accumulate (2)"(In 1=@WavePlayer_0.On Finished, In 0=@WavePlayer_1.On Finished)
  ]
  links [
   {dst="@UE.OutputFormat.Stereo.Audio:0", src=Add_0.Out}
   {dst="@UE.OutputFormat.Stereo.Audio:1", src=Add_1.Out}
   {dst="@UE.Source.OneShot.OnFinished", src=TriggerAccumulatorTriggerAccumulate2_0.Out}
  ]
 }
 Mobile {
  nodes [
   n1 = Clamp.Clamp:Float(In=UE.Divide:Float(PrimaryOperand=^HitMarker.Azimuth) {also {defaults {AdditionalOperands=90}}}.Out) {also {defaults {Max=1, Min=-1}}}
   TriggerDelay_0 = UE.Trigger Delay(Delay Time=^Patch.OptionalDelay, In=^UE.Source.OnPlay)
   ArrayRandomGet_0 = Array.Random Get:WaveAsset:Array(In Array=^LayerB.Wave, Next=@TriggerDelay_0.Out)
   n2 = MapRange.MapRange:Float(In=^HitMarker.InstigatorAccumulatedDamage) {also {defaults {In Range B=300}}}
   WavePlayer_0 = UE.Wave Player:Mono(
    Play=@ArrayRandomGet_0.On Next
    Wave Asset=@ArrayRandomGet_0.Value
    Pitch Shift=MapRange.MapRange:Float(Out Range B=^LayerB.Pitch.Max, Out Range A=^LayerB.Pitch.Min, In=@n2.Out Value).Out Value
   ) {also {version [1,1]}}
   ArrayRandomGet_1 = Array.Random Get:WaveAsset:Array(In Array=^LayerA.Wave, Next=@TriggerDelay_0.Out)
   WavePlayer_1 = UE.Wave Player:Mono(
    Play=@ArrayRandomGet_1.On Next
    Wave Asset=@ArrayRandomGet_1.Value
    Pitch Shift=MapRange.MapRange:Float(Out Range B=^LayerA.Pitch.Max, Out Range A=^LayerA.Pitch.Min, In=@n2.Out Value).Out Value
   ) {also {version [1,1]}}
   AudioMixerAudioMixerMono2_0 = "AudioMixer.Audio Mixer (Mono, 2)"(Gain 0=^LayerA.Gain, Gain 1=^LayerB.Gain, In 1=@WavePlayer_0.Out Mono, In 0=@WavePlayer_1.Out Mono)
   DecibelstoLinearGain_0 = UE.Decibels to Linear Gain:Float() {also {defaults {Decibels=-3}}}
   n3 = TriggerCompare.Clamp:Float(A=^Time.TimeSinceLastPlayed, Compare=^UE.Source.OnPlay) {also {defaults {B=0}}}
   n4 = TriggerCompare.Clamp:Float(B=^Retrigger.BehaviorThresholdTime, A=^Time.TimeSinceLastPlayed, Compare=@n3.False) {also {defaults {Type=3}}}
   TriggerAnyTriggerAny2_0 = "TriggerAny.Trigger Any (2)"(In 1=@n4.True, In 0=@n3.True)
   TriggerRouteTriggerRouteFloat2_0 = "TriggerRoute.Trigger Route (Float, 2):Float"(Value 1=^Retrigger.FadeInTime, Set 0=@TriggerAnyTriggerAny2_0.Out, Set 1=@n4.False) {also {defaults {Value 0=0}}}
   n5 = &MSP_Fade(In=@TriggerRouteTriggerRouteFloat2_0.On Set, Fade Time=@TriggerRouteTriggerRouteFloat2_0.Value)
   n6 = MapRange.MapRange:Float(In=^HitMarker.DamageDealt, Out Range B=^Patch.DamageDealthVolModifier.Max, Out Range A=^Patch.DamageDealthVolModifier.Min) {also {defaults {In Range B=100}}}
   n7 = UE.Multiply:Float(
    AdditionalOperands=^Patch.Gain
    PrimaryOperand=UE.Multiply:Float(
     AdditionalOperands=Clamp.Clamp:Float(Max=^Patch.DamageDealthVolModifier.ReClamp.Max, Min=^Patch.DamageDealthVolModifier.ReClamp.Min, In=@n6.Out Value).Value
     PrimaryOperand=Modulation.GetModulatorValue(Modulator=^Patch.Modulator.Volume).Out
    ).Out
   )
   n8 = UE.Multiply:Audio by Float(
    PrimaryOperand=@AudioMixerAudioMixerMono2_0.Out
    AdditionalOperands=UE.Multiply:Float(
     AdditionalOperands=@DecibelstoLinearGain_0.Linear Gain
     PrimaryOperand=UE.Multiply:Float(AdditionalOperands=@n5.Value Out, PrimaryOperand=@n7.Out).Out
    ).Out
   )
   StereoPanner_0 = UE.Stereo Panner(
    Pan Amount=Clamp.Clamp:Float(In=UE.Add:Float(AdditionalOperands=^Patch.StereoSpread, PrimaryOperand=@n1.Value).Out) {also {defaults {Max=1, Min=-1}}}.Value
    In=@n8.Out
   )
   StereoPanner_1 = UE.Stereo Panner(
    Pan Amount=Clamp.Clamp:Float(In=UE.Subtract:Float(AdditionalOperands=^Patch.StereoSpread, PrimaryOperand=@n1.Value).Out) {also {defaults {Max=1, Min=-1}}}.Value
    In=@n8.Out
   )
   Add_0 = UE.Add:Audio(AdditionalOperands=@StereoPanner_0.Out Left, PrimaryOperand=@StereoPanner_1.Out Left)
   Add_1 = UE.Add:Audio(AdditionalOperands=@StereoPanner_0.Out Right, PrimaryOperand=@StereoPanner_1.Out Right)
   TriggerAccumulatorTriggerAccumulate2_0 = "TriggerAccumulator.Trigger Accumulate (2)"(In 1=@WavePlayer_0.On Finished, In 0=@WavePlayer_1.On Finished)
  ]
  links [
   {dst="@UE.OutputFormat.Stereo.Audio:0", src=Add_0.Out}
   {dst="@UE.OutputFormat.Stereo.Audio:1", src=Add_1.Out}
   {dst="@UE.Source.OneShot.OnFinished", src=TriggerAccumulatorTriggerAccumulate2_0.Out}
  ]
 }
]
properties {
 assetImportData=&self:AssetImportData
 audioPropertiesSheet=&/APS_Hitmarker_Instigator_Standard
 numChannels=2
 outputFormat=Stereo
 sampleRate=48000
 soundClassObject=&Fort_Hitmarkers
 soundSubmixObject=&Fortnite_SFX
 busSends [
  {audioBus=&AB_Sidechain_FortWeapon_Vol, sendLevel=0.35}
  {audioBus=&AB_Sidechain_FortExplosion_Vol, sendLevel=0.35}
  {audioBus=&AB_Sidechain_FortAmbient_Vol, sendLevel=0.3}
  {audioBus=&AB_Sidechain_FortLoot_Vol, sendLevel=0.3}
  {audioBus=&AB_Sidechain_FortVehicles_Vol, sendLevel=0.3}
  {audioBus=&AB_Sidechain_FortCreatures_Vol, sendLevel=0.3}
  {audioBus=&AB_Sidechain_FortEmote_Vol, sendLevel=0.3}
 ]
 concurrencySet [&/CON_Hitmarker_GunplayUpdate_Instigator_Standard]
 modulationSettings {
  volumeRouting=Union
  volumeModulationDestination {bEnableModulation=true, modulators [&Patch_FortHitNotify_Vol]}
 }
}
