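# Stereo one-shot MetaSound preset that plays the observed hit notification for a destroyed shield.
# It takes its graph from MSS_HitNotifies_GunplayUpdate_Parent, feeding a 3D shield-break wave to Layer A and leaving Layer B without waves.
# Inputs expose hit-marker data, damage-based volume, pitch tables and retrigger timing.
# It uses Observed attenuation but Instigator concurrency and property sheet, a Receiver volume modulator, and sends to sidechain buses.
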MetaSoundSource /Game/Athena/Sounds/HitNotifications/Gunplay_Update/MSS_HitNotifies_GunplayUpdate_Shield_Destroyed_Observed
preset_parent=&/MSS_HitNotifies_GunplayUpdate_Parent
inputs [
 HitMarker.Azimuth {type=Float}
 HitMarker.DamageDealt {type=Float}
 HitMarker.Distance {type=Float}
 HitMarker.InstigatorAccumulatedDamage {type=Float}
 HitMarker.ReceiverPercentHealth {type=Float}
 HitMarker.ReceiverPercentShield {type=Float}
 LayerA.Gain {type=Float}
 LayerA.Pitch.Max {type=Float}
 LayerA.Pitch.Min {type=Float}
 LayerA.PitchTable {type=WaveTableBankAsset}
 LayerA.Wave {type=WaveAsset:Array}
 LayerB.Gain {type=Float}
 LayerB.Pitch.Max {type=Float}
 LayerB.Pitch.Min {type=Float}
 LayerB.PitchTable {type=WaveTableBankAsset}
 LayerB.Wave {type=WaveAsset:Array}
 Patch.DamageDealthVolModifier.Max {type=Float}
 Patch.DamageDealthVolModifier.Min {type=Float}
 Patch.DamageDealthVolModifier.ReClamp.Max {type=Float}
 Patch.DamageDealthVolModifier.ReClamp.Min {type=Float}
 Patch.Gain {type=Float}
 Patch.Modulator.Volume {type=Modulator}
 Patch.OptionalDelay {type=Time}
 Patch.StereoSpread {type=Float}
 Retrigger.BehaviorThresholdTime {type=Float, access=Constructor}
 Retrigger.FadeInTime {type=Float, access=Constructor}
 Time.ElapsedTime {type=Float}
 Time.GameTimeSeconds {type=Float}
 Time.PlayTime {type=Float}
 Time.TimeSinceLastPlayed {type=Float}
 UE.Source.OnPlay {type=Trigger}
]
interfaces [Fort.HitMarker:1.0, Fort.Time:1.0, UE.OutputFormat.Stereo:1.0, UE.Source.OneShot:1.0, UE.Source:1.1]
outputs [
 "UE.OutputFormat.Stereo.Audio:0" {type=Audio}
 "UE.OutputFormat.Stereo.Audio:1" {type=Audio}
 UE.Source.OneShot.OnFinished {type=Trigger}
]
pages [Default {input_defaults {
   HitMarker.Azimuth=0
   HitMarker.DamageDealt=0
   HitMarker.Distance=0
   HitMarker.InstigatorAccumulatedDamage=0
   HitMarker.ReceiverPercentHealth=1
   HitMarker.ReceiverPercentShield=1
   LayerA.Gain=1
   LayerA.Pitch.Max=0
   LayerA.Pitch.Min=0
   LayerA.PitchTable=&/WB_Hitmarker_GunplayUpdate
   LayerB.Gain=1.2
   LayerB.Pitch.Max=2
   LayerB.Pitch.Min=-1
   LayerB.PitchTable=&WB_Hitmarker_GunplayUpdate
   Patch.DamageDealthVolModifier.Max=2
   Patch.DamageDealthVolModifier.Min=0.6
   Patch.DamageDealthVolModifier.ReClamp.Max=1.2
   Patch.DamageDealthVolModifier.ReClamp.Min=0
   Patch.Gain=0.65
   Patch.Modulator.Volume=/Game/Athena/Sounds/HitNotifications/Gunplay_Update/Modulation/Patch_Hitmarkers_Receiver_Standard_Volume
   Patch.OptionalDelay=0
   Patch.StereoSpread=1
   Retrigger.BehaviorThresholdTime=0.71
   Retrigger.FadeInTime=0.1
   Time.ElapsedTime=0
   Time.GameTimeSeconds=0
   Time.PlayTime=0
   Time.TimeSinceLastPlayed=0
   UE.Source.OnPlay=null
   LayerA.Wave [/Game/Athena/Sounds/HitNotifications/HitMarker_ShieldBreak_3D_01]
   LayerB.Wave []
  }}]
properties {
 assetImportData=&self:AssetImportData
 attenuationSettings=/Game/Athena/Sounds/HitNotifications/Fort_HitMarker_Observed_Att
 audioPropertiesSheet=&/APS_Hitmarker_Instigator_Standard
 maxDistance=5500
 numChannels=2
 outputFormat=Stereo
 sampleRate=48000
 soundClassObject=&Fort_Hitmarkers
 soundSubmixObject=&Fortnite_SFX
 busSends [
  {audioBus=&AB_Sidechain_FortWeapon_Vol, sendLevel=0.35}
  {audioBus=&AB_Sidechain_FortExplosion_Vol, sendLevel=0.35}
  {audioBus=&AB_Sidechain_FortAmbient_Vol, sendLevel=0.3}
  {audioBus=&AB_Sidechain_FortLoot_Vol, sendLevel=0.3}
  {audioBus=&AB_Sidechain_FortVehicles_Vol, sendLevel=0.3}
  {audioBus=&AB_Sidechain_FortCreatures_Vol, sendLevel=0.3}
  {audioBus=&AB_Sidechain_FortEmote_Vol, sendLevel=0.3}
 ]
 concurrencySet [&/CON_Hitmarker_GunplayUpdate_Instigator_Standard]
 modulationSettings {
  volumeRouting=Union
  volumeModulationDestination {bEnableModulation=true, modulators [&Patch_FortHitNotify_Vol]}
 }
}
